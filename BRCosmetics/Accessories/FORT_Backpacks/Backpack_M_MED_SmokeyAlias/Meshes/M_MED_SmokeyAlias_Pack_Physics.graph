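# Physics asset for the SmokeyAlias cosmetic backpack, giving its dangling and hinged parts capsule bodies for secondary motion.
# Skull, four lids and two locks constrain directly to the pack body, with tight limits on lids and wide swing on locks.
# A three-link jaw and two cloth strands chain outward from the pack, with selected swing or twist axes locked.
# Three large spine capsules have no constraints, apparently serving as body collision volumes.
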PhysicsAsset /BRCosmetics/Accessories/FORT_Backpacks/Backpack_M_MED_SmokeyAlias/Meshes/M_MED_SmokeyAlias_Pack_Physics
bodies {
 dyn_cloth_10_joint [{shapeType=Capsule, length=2, radius=0.5, shapeName=dyn_cloth_10_joint_capsule, center=<2.1242 0 0>, rotation=<rot 90 166.7373 166.7371>}]
 dyn_cloth_1_joint [{shapeType=Capsule, length=1, radius=2, shapeName=dyn_cloth_1_joint_capsule, center=<3.0106 0 0>, rotation=<rot 90 0 -0.0001>}]
 dyn_cloth_2_joint [{shapeType=Capsule, length=1.5, radius=1.5, shapeName=dyn_cloth_2_joint_capsule, center=<2.4903 0 0>, rotation=<rot 90 6.7213 6.7211>}]
 dyn_cloth_3_joint [{shapeType=Capsule, length=1.5, radius=1, shapeName=dyn_cloth_3_joint_capsule, center=<2.0352 0 0>, rotation=<rot 90 -179.9993 -179.9995>}]
 dyn_cloth_4_joint [{shapeType=Capsule, length=1, radius=1, shapeName=dyn_cloth_4_joint_capsule, center=<1.5486 0 0>, rotation=<rot 90 0.0001 -0.0001>}]
 dyn_cloth_5_joint [{shapeType=Capsule, length=1.5, radius=1, shapeName=dyn_cloth_5_joint_capsule, center=<1.8658 0 0>, rotation=<rot 90 179.9998 179.9996>}]
 dyn_cloth_7_joint [{shapeType=Capsule, length=2, radius=0.5, shapeName=dyn_cloth_7_joint_capsule, center=<2.0719 0 0>, rotation=<rot 90 0.0132 0.0131>}]
 dyn_cloth_8_joint [{shapeType=Capsule, length=2, radius=0.5, shapeName=dyn_cloth_8_joint_capsule, center=<2.217 0 0>, rotation=<rot 90 90 89.9998>}]
 dyn_cloth_9_joint [{shapeType=Capsule, length=2, radius=0.5, shapeName=dyn_cloth_9_joint_capsule, center=<1.9587 0 0>, rotation=<rot 90 179.9605 179.9603>}]
 dyn_lid_l_down_joint [{shapeType=Capsule, length=5.5, radius=0.5, shapeName=dyn_lid_l_down_joint_capsule, center=<0.7666 0 0>, rotation=<rot 90 -179.995 -179.9953>}]
 dyn_lid_l_up_joint [{
   shapeType=Capsule
   length=3.5
   radius=0.5
   shapeName=dyn_lid_l_up_joint_capsule
   center=<1.5155 0 -0.1326>
   rotation=<rot 85 -0.0029 -0.0027>
  }]
 dyn_lid_r_down_joint [{shapeType=Capsule, length=5, radius=0.5, shapeName=dyn_lid_r_down_joint_capsule, center=<0.586 0 0.1054>, rotation=<rot -90 180 -179.9994>}]
 dyn_lid_r_up_joint [{shapeType=Capsule, length=3.5, radius=0.5, shapeName=dyn_lid_r_up_joint_capsule, center=<0.5161 0 0>, rotation=<rot -90 180 -180>}]
 dyn_lock_l_joint [{shapeType=Capsule, length=1, radius=0.5, shapeName=dyn_lock_l_joint_capsule, center=<1.4311 0 0>, rotation=<rot -90 -90 90.0004>}]
 dyn_lock_r_joint [{shapeType=Capsule, length=1, radius=0.5, shapeName=dyn_lock_r_joint_capsule, center=<1.3861 0 0>, rotation=<rot -90 89.2264 -84.2259>}]
 dyn_skull_jaw_1_joint [{shapeType=Capsule, length=13.5, radius=3, shapeName=dyn_skull_jaw_1_joint_capsule, center=<-1.5993 0 0>}]
 dyn_skull_jaw_2_joint [{
   shapeType=Capsule
   length=9
   radius=1.5
   shapeName=dyn_skull_jaw_2_joint_capsule
   center=<-0.989 -0.0108 0>
   rotation=<rot 90 72.3756 77.3747>
  }]
 dyn_skull_jaw_3_joint [{shapeType=Capsule, length=3.5, radius=3.5, shapeName=dyn_skull_jaw_3_joint_capsule, center=<0 0 -1.0201>, rotation=<rot -90 0 -0.0004>}]
 dyn_skull_joint [{shapeType=Capsule, length=27, radius=3, shapeName=dyn_skull_joint_capsule, center=<0 -0.4397 -0.5241>, rotation=<rot 0 0 40>}]
 dyn_spine_02_joint [{shapeType=Capsule, length=19, radius=13.5, shapeName=dyn_spine_02_joint_capsule}]
 dyn_spine_03_joint [{shapeType=Capsule, length=20, radius=12.5, shapeName=dyn_spine_03_joint_capsule}]
 dyn_spine_04_joint [{shapeType=Capsule, length=20.5, radius=12, shapeName=dyn_spine_04_joint_capsule}]
 pack [{shapeType=Capsule, length=3, radius=1, shapeName=pack_capsule}]
}
constraints [
 {
  bone1Name=dyn_skull_joint
  bone2Name=pack
  swing1LimitDegrees=45
  swing1Motion=Locked
  swing2LimitDegrees=7
  twistLimitDegrees=45
  twistMotion=Locked
 }
 {bone1Name=dyn_lid_r_down_joint, bone2Name=pack, swing1LimitDegrees=1, swing2LimitDegrees=10, twistLimitDegrees=0}
 {bone1Name=dyn_lid_r_up_joint, bone2Name=pack, swing1LimitDegrees=1, swing2LimitDegrees=10, twistLimitDegrees=0}
 {bone1Name=dyn_lid_l_down_joint, bone2Name=pack, swing1LimitDegrees=1, swing2LimitDegrees=10, twistLimitDegrees=0}
 {bone1Name=dyn_lid_l_up_joint, bone2Name=pack, swing1LimitDegrees=1, swing2LimitDegrees=10, twistLimitDegrees=0}
 {
  bone1Name=dyn_cloth_1_joint
  bone2Name=pack
  swing1LimitDegrees=10
  swing2LimitDegrees=45
  swing2Motion=Locked
  twistLimitDegrees=0
  twistMotion=Locked
 }
 {bone1Name=dyn_cloth_7_joint, bone2Name=pack, swing1LimitDegrees=10, swing2LimitDegrees=45, swing2Motion=Locked, twistLimitDegrees=0}
 {bone1Name=dyn_cloth_8_joint, bone2Name=dyn_cloth_7_joint, swing1LimitDegrees=20, swing2LimitDegrees=6, twistLimitDegrees=5}
 {bone1Name=dyn_cloth_9_joint, bone2Name=dyn_cloth_8_joint, swing1LimitDegrees=30, swing2LimitDegrees=10, twistLimitDegrees=5}
 {bone1Name=dyn_cloth_10_joint, bone2Name=dyn_cloth_9_joint, swing1LimitDegrees=30, swing2LimitDegrees=10, twistLimitDegrees=5}
 {
  bone1Name=dyn_cloth_2_joint
  bone2Name=dyn_cloth_1_joint
  swing1LimitDegrees=20
  swing2LimitDegrees=6
  twistLimitDegrees=5
  twistMotion=Locked
 }
 {
  bone1Name=dyn_cloth_3_joint
  bone2Name=dyn_cloth_2_joint
  swing1LimitDegrees=30
  swing2LimitDegrees=12
  twistLimitDegrees=5
  twistMotion=Locked
 }
 {bone1Name=dyn_cloth_4_joint, bone2Name=dyn_cloth_3_joint, swing1LimitDegrees=30, swing2LimitDegrees=10, twistLimitDegrees=5}
 {bone1Name=dyn_cloth_5_joint, bone2Name=dyn_cloth_4_joint, swing1LimitDegrees=30, swing2LimitDegrees=15, twistLimitDegrees=5}
 {bone1Name=dyn_lock_l_joint, bone2Name=pack, swing1LimitDegrees=90, swing2LimitDegrees=5, twistLimitDegrees=3}
 {bone1Name=dyn_lock_r_joint, bone2Name=pack, swing1LimitDegrees=90, swing2LimitDegrees=5, twistLimitDegrees=3}
 {
  bone1Name=dyn_skull_jaw_1_joint
  bone2Name=pack
  swing1LimitDegrees=5
  swing2LimitDegrees=45
  swing2Motion=Locked
  twistLimitDegrees=45
  twistMotion=Locked
 }
 {
  bone1Name=dyn_skull_jaw_2_joint
  bone2Name=dyn_skull_jaw_1_joint
  swing1LimitDegrees=7
  swing2LimitDegrees=45
  swing2Motion=Locked
  twistLimitDegrees=45
  twistMotion=Locked
 }
 {
  bone1Name=dyn_skull_jaw_3_joint
  bone2Name=dyn_skull_jaw_2_joint
  swing1LimitDegrees=10
  swing2LimitDegrees=45
  swing2Motion=Locked
  twistLimitDegrees=45
  twistMotion=Locked
 }
]
