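# Animation Blueprint for the SmokeyAlias backpack cosmetic that follows the wearer's spine and adds physics-driven secondary motion.
# Each update caches world transforms of spine_01 to spine_03 from the attach parent.
# These replace dyn_spine_02 to dyn_spine_04 joint transforms in world space, feeding a RigidBody simulation based on the pack bone using the pack's physics asset.
# An additive roll then tilts dyn_skull_joint while skydiving.
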AnimBlueprint /BRCosmetics/Accessories/FORT_Backpacks/Backpack_M_MED_SmokeyAlias/Meshes/M_MED_SmokeyAlias_Pack_AnimBP
target_skeleton=&/M_MED_SmokeyAlias_Pack_Skeleton
graphs [
 AnimGraph {nodes [
   n1 = Get!() -> (transform spine_03_Location:Vector, transform spine_03_Rotation:Rotator, transform spine_03_Scale:Vector)
   n2 = "Transform (Modify) Bone - Bone: dyn_spine_04_joint"(
    ComponentPose=Local To Component(LocalPose=Input Pose())
    Alpha=1
    Translation=@n1
    Rotation=@n1.transform spine_03_Rotation
    Scale=@n1.transform spine_03_Scale
   ) {
    rotationMode=BMM_Replace
    rotationSpace=BCS_WorldSpace
    scaleMode=BMM_Replace
    scaleSpace=BCS_WorldSpace
    translationMode=BMM_Replace
    translationSpace=BCS_WorldSpace
    alphaBoolBlend {blendOption=Linear}
    boneToModify {boneName=dyn_spine_04_joint}
   }
   n3 = Get!() -> (transform spine_02_Location:Vector, transform spine_02_Rotation:Rotator, transform spine_02_Scale:Vector)
   n4 = "Transform (Modify) Bone - Bone: dyn_spine_03_joint"(ComponentPose=@n2, Alpha=1, Translation=@n3, Rotation=@n3.transform spine_02_Rotation, Scale=@n3.transform spine_02_Scale) {
    rotationMode=BMM_Replace
    rotationSpace=BCS_WorldSpace
    scaleMode=BMM_Replace
    scaleSpace=BCS_WorldSpace
    translationMode=BMM_Replace
    translationSpace=BCS_WorldSpace
    alphaBoolBlend {blendOption=Linear}
    boneToModify {boneName=dyn_spine_03_joint}
   }
   n5 = Get!() -> (transform spine_01_Location:Vector, transform spine_01_Rotation:Rotator, transform spine_01_Scale:Vector)
   n6 = "Transform (Modify) Bone - Bone: dyn_spine_02_joint"(ComponentPose=@n4, Alpha=1, Translation=@n5, Rotation=@n5.transform spine_01_Rotation, Scale=@n5.transform spine_01_Scale) {
    rotationMode=BMM_Replace
    rotationSpace=BCS_WorldSpace
    scaleMode=BMM_Replace
    scaleSpace=BCS_WorldSpace
    translationMode=BMM_Replace
    translationSpace=BCS_WorldSpace
    alphaBoolBlend {blendOption=Linear}
    boneToModify {boneName=dyn_spine_02_joint}
   }
   n7 = RigidBody(ComponentPose=@n6) {
    overridePhysicsAsset=&/M_MED_SmokeyAlias_Pack_Physics
    simulationSpace=BaseBoneSpace
    alphaBoolBlend {blendOption=Linear}
    baseBoneRef {boneName=pack}
    componentAppliedLinearAccClamp=<3000 3000 3000>
    componentLinearAccScale=<0.1 0.1 0.1>
    componentLinearVelScale=<0.4 0.4 0.4>
    overrideWorldGravity {z=-980}
    simSpaceSettings {maxAngularAcceleration=3000, maxAngularVelocity=3000, maxLinearAcceleration=3000, maxLinearVelocity=3000, worldAlpha=0.75}
   }
   n8 = "Transform (Modify) Bone - Bone: dyn_skull_joint"!(ComponentPose=@n7, bAlphaBoolEnabled:bool=Get bIsSkyDiving(), Translation="0,0,0", Rotation="0,0,14", Scale="1,1,1") -> (Pose) {
    alphaInputType=Bool
    rotationMode=BMM_Additive
    alphaBoolBlend {blendOption=Linear}
    boneToModify {boneName=dyn_skull_joint}
    rotation {roll=14}
   }
   Output Pose(Result=Component To Local(ComponentPose=@n8)) {layerGroup=None}
  ]}
 EventGraph {nodes [
   Event BlueprintUpdateAnimation()
   n1 = GetAttachParent!(self:Scene Component Object Reference=GetPartSkeletalMeshComponent()) -> (ReturnValue:Scene Component Object Reference)
   > Set transform spine_03(transform spine_03:Transform=GetSocketTransform(self=@n1, InSocketName=spine_03, TransformSpace=RTS_World)) -> (Output_Get:Transform)
   > Set transform spine_02(transform spine_02:Transform=GetSocketTransform(self=@n1, InSocketName=spine_02, TransformSpace=RTS_World)) -> (Output_Get:Transform)
   > Set transform spine_01(transform spine_01:Transform=GetSocketTransform(self=@n1, InSocketName=spine_01, TransformSpace=RTS_World)) -> (Output_Get:Transform)
   TryGetPawnOwner()
  ]}
]
variables [
 Sky {type=boolean}
 transform spine_01 {type=Transform, default {location=<0 0 0>, rotation=<rot 0 0 0>, scale=<1 1 1>}}
 transform spine_02 {type=Transform, default {location=<0 0 0>, rotation=<rot 0 0 0>, scale=<1 1 1>}}
 transform spine_03 {type=Transform, default {location=<0 0 0>, rotation=<rot 0 0 0>, scale=<1 1 1>}}
]
